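# Translucent unlit master material that draws anime-style outlines on skeletal meshes and Niagara mesh particles.
# Four-tap scene depth edge detection, optionally combined with GBuffer roughness edges and masked by a subsurface GBuffer tag, drives opacity, with mobile/forward fallbacks.
# Emissive comes from an outline color, texture or darkened scene diffuse.
# WPO inflates the shell along normals, with optional bobbing and distance collapse; the FOV compensation nodes are unconnected.
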Material /Game/Effects/Niagara/MasterMaterials/ToonShader/M_AnimeOutline_FX
graph {
 properties {
  bEnableResponsiveAA=true
  bUseMaterialAttributes=true
  bUsedWithMeshDeformer=true
  bUsedWithNanite=true
  bUsedWithNiagaraMeshParticles=true
  bUsedWithSkeletalMesh=true
  blendMode=BLEND_Translucent
  pixelDepthOffsetMode=PDOM_Legacy
  shadingModel=MSM_Unlit
  twoSided=true
  usedShadingModels=Unlit
 }
 nodes [
  Constant() {r=2}
  Constant() {r=0.4}
  n1 = NamedRerouteUsage()
  n2 = ScalarParameter() {parameterName=SpecialDeviceFix}
  MaterialFunctionCall!(
   Material=SetMaterialAttributes!(
    MaterialAttributes
    Emissive Color=NamedRerouteUsage()
    Opacity=LinearInterpolate(A=@n1, Alpha=@n2) {constB=0}
    World Position Offset=NamedRerouteUsage()
   ) {attributeSetTypes [#1, #2, #3]}
   SkinFX_Mask
  ) {materialFunction=/Game/Global/MaterialFunctions/MF_ClothSimWPO_v2}
  n3 = ComponentMask(None=ViewProperty() {property=MEVP_TanHalfFieldOfView}.Property) {r=true}
  n4 = Divide(
   A=Multiply(A=Subtract(A=@n3), B=ScalarParameter() {defaultValue=0.5, group=FOV Compansation, parameterName=FOV Multi})
   B=MaterialFunctionCall(Multiplier=Constant() {r=2}) {materialFunction=&Pi}.Result
  )
  n5 = Add(A=Multiply(A=Multiply(A=@n4) {constB=-1}, B=ScalarParameter() {defaultValue=10, group=FOV Compansation, parameterName=Compansation}))
  MaterialFunctionCall(Number=@n5) {materialFunction=&DebugScalarValues}
  n6 = SceneTexture() {sceneTextureId=PPI_SubsurfaceColor}
  n7 = Constant()
  n8 = If(
   A=Min(A=ComponentMask(None=@n6.Color) {g=true}, B=ComponentMask(None=@n6.Color) {b=true})
   B=SquareRoot(None=ScalarParameter() {defaultValue=0.001, group=Cartoon, parameterName=Subsurface Mask})
   "A > B"=@n7
   "A == B"=Constant() {r=1}
   "A < B"=@n7
  ) {equalsThreshold=0.0005}
  n9 = Floor(None=SceneTexture() {sceneTextureId=PPI_Opacity}.Color) {desc="for effect quality low (SSColor is disabled)"}
  n10 = Constant()
  n11 = If(
   A=ComponentMask(None=SceneTexture() {sceneTextureId=PPI_ShadingModelColor}.Color) {b=true, g=true, r=true}
   B=Constant3Vector() {constant=<rgba 0.7 1 1 0>}
   "A > B"=@n10
   "A == B"=Constant() {r=1}
   "A < B"=@n10
  ) {equalsThreshold=0.001}
  n12 = Constant() {r=1}
  NamedRerouteDeclaration(None=ShadingPathSwitch(Default=@n8, Deferred=QualitySwitch(Default=@n8, Low=@n9, High=@n11, Medium=@n11, Epic=@n8), Forward=@n12, Mobile=@n12)) {name=Subsurface Mask, nodeColor=<rgba 0.5469 0 1 1>}
  n13 = Multiply(A=MaterialFunctionCall(Speed=ScalarParameter() {defaultValue=0.5, group=BobbingWPO, parameterName=RotatingSpeed}) {materialFunction=&DebugTimeSine}.Result) {constB=2}
  n14 = RotateAboutAxis(
   NormalizedRotationAxis=Constant3Vector() {constant=<rgba 0 0 1 0>}
   RotationAngle=Multiply(A=Subtract(A=@n13), B=ScalarParameter() {group=BobbingWPO, parameterName=RotatingAmount})
   PivotPoint=Constant()
   Position=MaterialFunctionCall() {materialFunction=&LocalPosition}.Local Position
  )
  n15 = Constant()
  n16 = ScalarParameter() {defaultValue=0.5, group=BobbingWPO, parameterName=BobbingSpeed}
  n17 = LinearInterpolate(
   B=Multiply(
    A=MaterialFunctionCall(Speed=@n16) {materialFunction=&DebugTimeSine}.Result
    B=ScalarParameter() {defaultValue=10, group=BobbingWPO, parameterName=BobbingVerticalHeight}
   ) {constB=10}
   Alpha=ScalarParameter() {group=BobbingWPO, parameterName=BobbingAmount}
  )
  NamedRerouteDeclaration(None=Add(
    A=Transform(None=@n14) {transformSourceType=TRANSFORMSOURCE_Local}
    B=MaterialFunctionCall(X=@n15, Y=@n15, Z=@n17) {materialFunction=&MakeFloat3}.Result
   )) {name=VerticalBobbing, nodeColor=<rgba 0.4766 0 1 1>}
  n18 = TextureSampleParameter2D() {
   group=Cartoon
   parameterName=OutlineMaskTexture
   samplerType=SAMPLERTYPE_Masks
   sortPriority=2
   texture=&T_BlackMask
  }
  n19 = OneMinus(None=@n18.R)
  n20 = SceneDepth() {desc=SceneDepth works on mobile whereas SceneTexture doesn't}
  n21 = PixelDepth()
  n22 = ScalarParameter() {defaultValue=50, group=Cartoon, parameterName=CRTNO_InnerLine_FadeStart, sliderMax=200}
  n23 = Saturate(None=Divide(
    A=Subtract(A=@n21.R, B=@n22)
    B=ScalarParameter() {defaultValue=500, group=Cartoon, parameterName=CRTNO_InnerLine_FadeDistance, sliderMax=1000}
   ))
  n24 = LinearInterpolate(
   A=ScalarParameter() {defaultValue=16, group=Cartoon, parameterName=SampleBias_Min, sliderMax=16}
   B=ScalarParameter() {defaultValue=16, group=Cartoon, parameterName=SampleBias, sliderMax=16}
   Alpha=@n23
  )
  n25 = LinearInterpolate(
   A=ScalarParameter() {defaultValue=1, group=Cartoon, parameterName=MobileSampleBias_Min, sliderMax=1}
   B=ScalarParameter() {defaultValue=16, group=Cartoon, parameterName=Mobile_SampleBias, sliderMax=16}
   Alpha=@n23
  )
  n26 = Divide(A=ObjectRadius(), B=ScalarParameter() {defaultValue=97, group=Cartoon, parameterName=ObjectScale, sliderMax=200})
  n27 = ComponentMask(None=ScreenPosition().ViewportUV) {g=true, r=true}
  n28 = PixelDepth()
  n29 = ScalarParameter() {defaultValue=50, group=Cartoon, parameterName=CRTNO_Outline-InnerScaleStart, sliderMax=200}
  n30 = Saturate(None=Divide(
    A=Subtract(A=@n28.R, B=@n29)
    B=ScalarParameter() {defaultValue=500, group=Cartoon, parameterName=CRTNO_Outline-InnerScaleDistance, sliderMax=1000}
   ))
  n31 = MaterialFunctionCall!(
   WorldUnits=LinearInterpolate(
    A=ScalarParameter() {defaultValue=0.07, group=Cartoon, parameterName=CRTNO_Outline-InnerSize1_WorldUnits, sliderMax=1}
    B=ScalarParameter() {defaultValue=1, group=Cartoon, parameterName=CRTNO_Outline-InnerSize2_WorldUnits, sliderMax=1}
    Alpha=@n30
   )
   Depth
  ) {materialFunction=/Engine/Functions/Engine_MaterialFunctions01/Coordinates/WorldUnitsInPixel}
  n32 = ScalarParameter() {defaultValue=0.51, group=Cartoon, parameterName=CRTNO_Outline-InnerSize1_MinPixels, sliderMax=1}
  n33 = ScalarParameter() {defaultValue=4, group=Cartoon, parameterName=CRTNO_Outline-InnerSize1_MaxPixels, sliderMax=8}
  n34 = Divide(A=ViewSize(), B=Constant() {r=320})
  n35 = Constant() {r=2}
  n36 = Clamp(None=Ceil(None=Subtract(A=@n34, B=@n35) {constB=0})) {desc="if resolution is bigger than 960 =2", maxDefault=2, minDefault=1}
  n37 = LinearInterpolate(
   A=ScalarParameter() {defaultValue=2, group=Cartoon, parameterName=CRTNO_Outline-InnerSize1}
   B=ScalarParameter() {defaultValue=0.75, group=Cartoon, parameterName=CRTNO_Outline-InnerSize2}
   Alpha=@n30
  )
  n38 = Multiply(A=@n37, B=ScalarParameter() {defaultValue=1.5, group=Cartoon, parameterName=CRTNO_Outline-InnerSizeMobileMultiplier})
  n39 = Divide(A=Constant() {r=960}, B=ComponentMask(None=ViewSize()) {r=true})
  n40 = Max(A=ShadingPathSwitch(Default=@n37, Forward=@n38, Mobile=@n38), B=Saturate(None=Floor(None=@n39)) {desc=if resolution is smaller than 960H}) {desc="If res smaller than target, calmp min value to 1"}
  n41 = Multiply(
   A=SceneTexelSize()
   B=StaticSwitchParameter(True=Clamp(None=@n31.PIxels, Min=@n32, Max=@n33), False=Multiply(A=@n36, B=@n40) {constA=2}) {group=Cartoon, parameterName=OutlineThickness_InWorldUnits, sortPriority=1}
  )
  n42 = Constant()
  n43 = AppendVector(A=ComponentMask(None=@n41) {desc=x, r=true}, B=@n42)
  n44 = Add(A=@n27, B=@n43) {desc=x+}
  n45 = AppendVector(A=@n42, B=ComponentMask(None=@n41) {desc=y, g=true}) {desc=y+}
  n46 = Add(A=@n27, B=@n45) {desc=y+}
  n47 = Constant() {r=-1}
  n48 = Add(A=@n27, B=Multiply(A=@n43, B=@n47)) {desc=x-}
  n49 = Add(A=@n27, B=Multiply(A=@n47, B=@n45)) {desc=y-}
  n50 = LinearInterpolate(
   A=ScalarParameter() {defaultValue=-1, group=Cartoon, parameterName=Divisor_Min, sliderMax=1, sliderMin=-1}
   B=ScalarParameter() {defaultValue=-16, group=Cartoon, parameterName=Divisor_max, sliderMax=16, sliderMin=-16}
   Alpha=@n23
  )
  n51 = ScalarParameter() {group=Cartoon, parameterName=Divisor_Offset}
  n52 = ScalarParameter() {defaultValue=-1, group=Cartoon, parameterName=MobileDivisor_Min, sliderMax=1, sliderMin=-1}
  n53 = ScalarParameter() {defaultValue=-16, group=Cartoon, parameterName=MobileDivisor_max, sliderMax=16, sliderMin=-16}
  n54 = MaterialFunctionCall!(
   Base=@n20.R
   BaseSampleBias=Multiply(A=ShadingPathSwitch(Default=@n24, Mobile=@n25), B=@n26)
   Scale=ScalarParameter() {defaultValue=4, group=Cartoon, parameterName=Scale, sliderMax=8}
   Sample1=SceneDepth(UVs=@n44).R
   Sample2=SceneDepth(UVs=@n46).R
   Sample3=SceneDepth(UVs=@n48).R
   Sample4=SceneDepth(UVs=@n49).R
   Divisor=ShadingPathSwitch(Default=Add(A=@n50, B=@n51), Mobile=LinearInterpolate(A=@n52, B=@n53, Alpha=@n23))
   UseCenteredLines=Constant()
   UseRGB
  ) {materialFunction=/Game/Packages/jl_tests/combine4Samples}
  n55 = StaticSwitchParameter(True=Multiply(A=@n54.Result, B=@n19), False=@n54.Result) {
   desc=Exclude a section of the mesh from having Distance Outlines based on a Texture
   group=Cartoon
   parameterName=UseMaskTextureForDistanceOutlines
   sortPriority=1
  }
  NamedRerouteDeclaration(None=@n55) {name=DistanceOutlines, nodeColor=<rgba 0.2188 0 1 1>}
  n56 = SceneTexture() {sceneTextureId=PPI_Roughness}
  n57 = Constant()
  n58 = Constant() {r=4}
  n59 = SceneTexture(UVs=@n44) {sceneTextureId=PPI_Roughness}
  n60 = SceneTexture(UVs=@n46) {sceneTextureId=PPI_Roughness}
  n61 = SceneTexture(UVs=@n48) {sceneTextureId=PPI_Roughness}
  n62 = SceneTexture(UVs=@n49) {sceneTextureId=PPI_Roughness}
  n63 = ScalarParameter() {defaultValue=-0.04, group=Cartoon_RoughnessGBufferLines, parameterName=InnerLinesDivisor, sliderMax=1, sliderMin=-1}
  n64 = ComponentMask(None=MaterialFunctionCall!(Roughness input=MaterialFunctionCall(
     Base=@n56.Color
     BaseSampleBias=@n57
     Scale=@n58
     Sample1=@n59.Color
     Sample2=@n60.Color
     Sample3=@n61.Color
     Sample4=@n62.Color
     Divisor=@n63
    ) {materialFunction=&combine4Samples}.Result) {materialFunction=/Game/Packages/Fortress_SharedMaterials/Anime_Master_Material/MF_OrientationFade}.Result) {r=true}
  n65 = StaticSwitchParameter(True=Multiply(A=@n19, B=@n64), False=@n64) {
   desc=Exclude a section of the mesh from having Distance Outlines based on a Texture
   group=Cartoon
   parameterName=UseMaskTextureForRoughnessOutlines
   sortPriority=1
  }
  NamedRerouteDeclaration(None=@n65) {name=RoughnessOutlines, nodeColor=<rgba 0.2188 0 1 1>}
  n66 = NamedRerouteUsage()
  n67 = Multiply(A=VertexNormalWS(), B=ScalarParameter() {defaultValue=0.5, group=Cartoon, parameterName=ScaleMesh, sliderMax=1})
  n68 = ScalarParameter() {defaultValue=2500, parameterName=Far Camera Fade Distance}
  n69 = LinearInterpolate(
   A=StaticSwitchParameter(True=Add(A=@n66, B=@n67), False=@n67) {group=BobbingWPO, parameterName=Use BobbingWPO, sortPriority=1}
   B=Subtract(A=ObjectPositionWS(), B=WorldPosition().XYZ)
   Alpha=LinearInterpolate(
    A=ScalarParameter() {group=S22ChromeCharacters, parameterName=S22ChromeCharacters_strength, sliderMax=1}
    B=MaterialFunctionCall(Distance=@n68) {materialFunction=&Near_Camera_Fade}.Result
    Alpha=ScalarParameter() {parameterName=Fade Outline at Distance}
   )
  )
  NamedRerouteDeclaration(None=StaticSwitchParameter(
    True=MaterialFunctionCall!(WPO=@n69) {materialFunction=/Game/Packages/Fortress_SharedMaterials/Base_Material_Functions/Fortnite_HideElementsWithVertexColor}.Result
    False=@n69
   ) {group=Opacity Mask, parameterName=Use Vertex Colors for Mask}) {name=WPO, nodeColor=<rgba 0 1 0.2734 1>}
  n70 = Subtract(A=PixelDepth().R, B=@n20.R)
  n71 = Multiply(A=ScalarParameter() {defaultValue=3, group=Cartoon, parameterName=CRTNO_Outline-InnerMaskBias, sliderMax=4}, B=@n26)
  n72 = ComponentMask(None=NamedRerouteUsage()) {r=true}
  n73 = Max(A=NamedRerouteUsage(), B=@n72)
  n74 = Multiply(
   A=Saturate(None=Add(A=@n70, B=@n71))
   B=StaticSwitchParameter(True=ShadingPathSwitch(Default=@n73, Forward=@n72, Mobile=@n72), False=@n72) {group=Cartoon_RoughnessGBufferLines, parameterName=UseRoughnessGBufferOutlines, sortPriority=1}
  )
  n75 = Multiply(A=NamedRerouteUsage(), B=@n74)
  n76 = Multiply(
   A=Saturate(None=StaticSwitchParameter(True=@n75, False=@n74) {defaultValue=true, group=Cartoon, parameterName=UseSubsurfaceMasking, sortPriority=1})
   B=ScalarParameter() {defaultValue=1, group=Cartoon, parameterName=CRTNO_Outline-InnerAmount, sliderMax=1}
  )
  n77 = Constant()
  n78 = ShadingPathSwitch(Default=@n76, Forward=@n77, Mobile=@n77)
  n79 = MaterialFunctionCall!(Opacity=StaticSwitchParameter(True=Constant() {r=1}, False=StaticSwitchParameter(True=@n76, False=@n78) {group=Cartoon, parameterName=UseDepthBufferOutlinesOnMobile}) {group=CartoonDebug, parameterName="Show Bodybag (For Debugging)", sortPriority=1}) {materialFunction=/Game/Packages/Fortress_SharedMaterials/Base_Material_Functions/BaseCharacterShaderFunctions/MF_TechArt_ChromeOutlineOverride}
  NamedRerouteDeclaration(None=@n79.Result) {name=Opacity, nodeColor=<rgba 0.6641 0 1 1>}
  n80 = TextureSampleParameter2D() {group=Cartoon, parameterName=OutlineDiffuseTexture, sortPriority=2, texture=&T_BlacksRGB}
  n81 = Multiply(
   A=@n80.RGB
   B=ScalarParameter() {defaultValue=1, desc=Multiplier for the Outline Diffuse Texture, group=Cartoon, parameterName=DiffuseMult_Outline}
  )
  n82 = VectorParameter() {group=Cartoon, parameterName=OutlineColor, sortPriority=3, defaultValue=<rgba 0 0 0 1>}
  n83 = StaticSwitchParameter(
   True=MaterialFunctionCall!(Color=@n82.RGB) {materialFunction=/Game/Packages/Fortress_SharedMaterials/Anime_Master_Material/MF_TODColorCurve}.Result
   False=@n82.RGB
  ) {group=Cartoon, parameterName=Outline Color Use TOD, sortPriority=1}
  n84 = VectorParameter() {group=Cartoon, parameterName=OutlineDakenMultiplier, sortPriority=3, defaultValue=<rgba 0.1 0.1 0.1 0>}
  n85 = Multiply(A=ComponentMask(None=SceneTexture() {sceneTextureId=PPI_DiffuseColor}.Color) {b=true, g=true, r=true}, B=@n84.RGB)
  n86 = VectorParameter() {group=Cartoon, parameterName=MobileOutlineColor, sortPriority=3}
  n87 = StaticSwitchParameter(
   True=MaterialFunctionCall(Color=@n86.RGB) {materialFunction=&MF_TODColorCurve}.Result
   False=@n86.RGB
  ) {group=Cartoon, parameterName=Outline Color Use TOD, sortPriority=1}
  n88 = StaticSwitchParameter(
   True=StaticSwitchParameter(
    True=MaterialFunctionCall(Color=@n81) {materialFunction=&MF_TODColorCurve}.Result
    False=@n81
   ) {group=Cartoon, parameterName=Outline Color Use TOD, sortPriority=1}
   False=ShadingPathSwitch(
    Default=StaticSwitchParameter(True=@n83, False=@n85) {desc=Use Vector Color for anime outline Color, group=Cartoon, parameterName=UseOutlineColor, sortPriority=1}
    Forward=@n87
    Mobile=@n87
   )
  ) {
   desc=Enable to use a texture to color the edge outlines of the mesh.
   group=Cartoon
   parameterName=UseTextureForOutlineDiffuse
   sortPriority=1
  }
  NamedRerouteDeclaration(None=StaticSwitchParameter(
    True=Fresnel()
    False=MaterialFunctionCall!(In=@n88, LUT, LUT Multiplier) {materialFunction=/Game/Packages/Fortress_SharedMaterials/Anime_Master_Material/MF_ApplyLUT}.Result
   ) {group=CartoonDebug, parameterName="Show Bodybag (For Debugging)", sortPriority=1}) {name=Emissive, nodeColor=<rgba 0.6641 1 0 1>}
  OneMinus(None=@n5)
  n89 = MaterialFunctionCall(Speed=Constant() {r=0.333}) {materialFunction=&DebugTimeSine}
  StaticSwitchParameter(True=LinearInterpolate(B=Multiply(A=@n5), Alpha=Round(None=@n89.Result))) {group=FOV Compansation, parameterName=UseFOVCompansation}
 ]
}
