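# Masked unlit Niagara master material for toon-style smoke on sprite and mesh particles.
# Particle color is banded by a rounded view-facing outline and a fake-light shadow, optionally blended flat, then passed through emissive adjustments.
# World-aligned cloud normals, panned by time and dynamic parameters, blend with vertex normals into a view-facing height that drives a HeightLerp erosion mask from particle alpha.
# A Fresnel depth-fade branch also scales alpha.
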Material /Game/Effects/Niagara/MasterMaterials/ToonShader/M_ToonSmoke_Simple
graph {
 properties {
  bAutomaticallySetUsageInEditor=false
  bUsedWithNiagaraMeshParticles=true
  bUsedWithNiagaraSprites=true
  blendMode=BLEND_Masked
  shadingModel=MSM_Unlit
  usedShadingModels=Unlit
 }
 nodes [
  n1 = ParticleColor()
  n2 = StaticBool() {value=true}
  n3 = ScalarParameter() {defaultValue=1, parameterName=FresnelPower}
  DepthFade(
   Opacity=Multiply(
    A=@n1.A
    B=MaterialFunctionCall(Invert Fresnel=@n2, Power=@n3) {materialFunction=&Fresnel_Function}.Result
   )
   FadeDistance=ScalarParameter() {defaultValue=25, parameterName=DepthFadeDistance}
  )
  Multiply(A=@n1.A, B=NamedRerouteUsage())
  n4 = Constant3Vector() {constant=<rgba 500 500 500 0>}
  n5 = DynamicParameter() {defaultValue=<rgba 0 0 0 1>, paramNames [Texture Size, ControlShadowEdge, PanTexture, ControlBloom]}
  n6 = Multiply(A=Time(), B=Multiply(A=@n5.PanTexture, B=Constant3Vector() {constant=<rgba 0 0 300 0>})) {constB=200}
  n7 = Add(A=@n6, B=WorldPosition().XYZ)
  n8 = ScalarParameter() {defaultValue=1, parameterName=UsePannerWPO?}
  n9 = LinearInterpolate(
   A=MaterialFunctionCall!(
    Diffuse Texture Object=TextureObject() {samplerType=SAMPLERTYPE_LinearColor, texture=/Game/Global/Textures/Default/Tileable_Patterns/Cloud_Diffuse}
    Spec Texture Object
    TextureSize=Multiply(A=@n4, B=@n5.Texture Size) {constB=3}
    --------------------------
    Use seperate settings for Z projection
    Use seperate texture for Z projection
    Texture Size Z projection
    Z Plane Diffuse Texture Object
    Z Plane Spec Texture Object
    "--------------------------      "
    ProjectionTransitionContrast
    Use Texture Alpha Heightmap to Refine the Transistion
    HeightLerpTransitionContrast
    PickTheDiffuseHeightChannel
    TheDiffuseHeightChannel
    "--------------------------       "
    NormalTextureObject=TextureObject() {samplerType=SAMPLERTYPE_Normal, texture=/Game/Global/Textures/Default/Tileable_Patterns/Cloud_normals}
    World Space Normals
    "--------------------------  "
    Export Float 4 Diffuse
    Export Float 4 Spec
    "--------------------------     "
    WorldPosition=LinearInterpolate(A=@n6, B=@n7, Alpha=@n8)
    World Space Normal
   ) {materialFunction=/Engine/Functions/Engine_MaterialFunctions02/Texturing/WorldAlignedTextures_Complex}.Normal XYZ Texture
   B=VertexNormalWS()
   Alpha=Constant() {desc=.2, r=0.2}
  ) {desc=Soft normal}
  n10 = CameraVectorWS()
  n11 = VertexNormalWS()
  n12 = CameraVectorWS()
  n13 = Add(A=Clamp(None=DotProduct(A=@n9, B=@n10)), B=Clamp(None=DotProduct(A=@n11, B=@n12)))
  n14 = ScalarParameter() {defaultValue=2, parameterName=BillowControl}
  NamedRerouteDeclaration(None=MaterialFunctionCall(Transition Phase=ParticleColor().A, Height Texture=Divide(A=@n13, B=@n14)) {materialFunction=&HeightLerp}.Alpha) {name=Mask Erosion, nodeColor=<rgba 0.1016 0 1 1>}
  n15 = VertexNormalWS()
  n16 = CameraVectorWS()
  n17 = Round(None=Power(Base=DotProduct(A=@n15, B=@n16), Exp=ScalarParameter() {defaultValue=0.6, parameterName=OutLineWidth}) {constExponent=0.7})
  n18 = ScalarParameter() {defaultValue=0.25, parameterName=ShadowIntensity}
  n19 = VectorParameter() {parameterName=LightDirection, defaultValue=<rgba 0.5 0 1 0>}
  n20 = ScalarParameter() {parameterName=ShadowOffset}
  n21 = DotProduct(A=VertexNormalWS(), B=Add(A=CameraVectorWS(), B=Add(A=@n19.RGB, B=@n20) {constB=0}))
  n22 = ScalarParameter() {defaultValue=5, parameterName=ShadowPowerRange}
  n23 = Round(None=Saturate(None=Power(Base=@n21, Exp=@n22) {constExponent=1}))
  n24 = Multiply(A=@n1.RGB, B=Multiply(A=Saturate(None=@n17), B=LinearInterpolate(A=@n18, Alpha=@n23) {constA=0.25}))
  n25 = ScalarParameter() {parameterName=Opaque or Soft Smoke}
  NamedRerouteDeclaration(None=MaterialFunctionCall(Emissive=LinearInterpolate(A=@n24, B=@n1.RGB, Alpha=@n25)) {materialFunction=&MF_FX_EmissiveAdjustments}.Result) {name=Emissive, nodeColor=<rgba 1 0 0.4687 1>}
 ]
}
